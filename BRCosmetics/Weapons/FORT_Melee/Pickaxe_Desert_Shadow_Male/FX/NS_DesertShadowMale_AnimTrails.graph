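# Niagara system for the swing trails of the Desert Shadow Male pickaxe, where only the Claw01 emitter is enabled.
# Claw01 uses shared FN AnimTrail modules to spawn three faint, short-lived, bone/socket-driven ribbon trails that narrow and fade.
# User.SecondImpact switches its alpha between two emitter-age flash curves, the second timed later in the brief emitter cycle.
# All other emitters (claw variants, soft trails, blade ribbons, debug sprite) are disabled.
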NiagaraSystem /BRCosmetics/Weapons/FORT_Melee/Pickaxe_Desert_Shadow_Male/FX/NS_DesertShadowMale_AnimTrails
dependencies {
 usedDataInterfaces [
  &NiagaraDataInterfaceArrayFloat3
  &NiagaraDataInterfaceCurve
  &NiagaraDataInterfaceColorCurve
  &NiagaraDataInterfaceVectorCurve
 ]
 usedDynamicInputs [&SetBoolByFloatComparison]
 usedModules [&SystemState, &Timeline]
}
emitters {
 Blades_Trails002 {
  bEnabled=false
  emitterUpdateScript [
   &EmitterState(Life Cycle Mode=::NewEnumerator1, Loop Behavior=::NewEnumerator1, Loop Duration=4, UseLoopDelay=true)
   &SpawnRate(SpawnRate=50, Use Spawn Probability=false)
  ]
  particleSpawnScript [
   &FN_SetSpawnGroup_v2(Number of Spawn Groups=2)
   /Niagara/Modules/Spawn/Initialization/InitializeParticle(
    Module.Write Color=true
    Module.Write Lifetime=true
    Module.Write Mass=true
    Module.Write Material Random=true
    Module.Write Position=false
    Module.Write Scale=false
    Module.Write SpriteRotation=false
    Module.Write SpriteSize=true
    Sprite Size=<25 25>
   ) {enabled=false}
   &InitializeRibbon(
    Lifetime=0.1
    Module.Write Color=true
    Module.Write Lifetime=true
    Module.Write Mass=true
    Module.Write Material Random=true
    Module.Write Position=true
    Module.Write RibbonTwist=false
    Module.Write RibbonWidth=true
    Ribbon Width=10
   )
   &SkeletalMeshLocation(
    Apply Sampled Orientation to Mesh Particles=true
    Bone / Socket Index=$Particles.SpawnGroup
    Bone Sampling Mode=::NewEnumerator4
    Inherit Velocity=false
    Interpolate Mesh Sampling=false
    MaskSpawnGroup=false
    Sampled Position Offset=&Add_Vector(
     A=&VectorFromCurve(
      VectorCurve=DataInterface() {propertyValues {XCurve {keys [<at 0 -25>,<at 1 -20>]}, YCurve {keys [<at 0 1>,<at 1 0>]}, ZCurve {keys [<at 0 1>,<at 1 0>]}}}
      Scale Curve=<0 0 0>
     )
     B=&SelectVectorFromArray(Direct Array Index=$Particles.SpawnGroup, Vector Selection Array=DataInterface() {propertyValues {FloatData [<30 0 0>, <-30 0 0>]}})
    )
    Sampled Position Offset Coordinate Space:ENiagaraMeshTransforms=::NewEnumerator3
    Skeletal Mesh=$User.NewNiagaraDataInterfaceSkeletalMesh
    Write Sampled Mesh Attrributes to Particle Attributes=true
   ) {enabled=false, module=SkeletalMeshLocation001}
   &SkeletalMeshLocation(
    Apply Sampled Orientation to Mesh Particles=true
    Inherit Velocity=false
    Interpolate Mesh Sampling=false
    MaskSpawnGroup=false
    Mesh Sampling Type=::NewEnumerator5
    Sampled Position Offset=&Add_Vector(
     A=&VectorFromCurve(
      VectorCurve=DataInterface() {propertyValues {XCurve {keys [<at 0 -25>,<at 1 -20>]}, YCurve {keys [<at 0 1>,<at 1 0>]}, ZCurve {keys [<at 0 1>,<at 1 0>]}}}
      Scale Curve=<0 0 0>
     )
     B=&SelectVectorFromArray(Direct Array Index=$Particles.SpawnGroup, Vector Selection Array=DataInterface() {propertyValues {FloatData [<30 0 0>, <-30 0 0>]}})
    )
    Sampled Position Offset Coordinate Space:ENiagaraMeshTransforms=::NewEnumerator3
    Skeletal Mesh=$User.SkeletalMesh
    Socket Sampling Mode=::NewEnumerator5
    Write Sampled Mesh Attrributes to Particle Attributes=true
   ) {module=SkeletalMeshLocation002}
   /SetVariables(Particles.Position=&Add_Vector(A=Linked() {linkedVariable {name=Particles.Position, type=NiagaraPosition}}, B=<0 0 -30>)) {isSetParameters=true}
   /SetVariables(Particles.PositionOffset1:Vector3f=&Add_Vector(A=Linked() {linkedVariable {name=Particles.Position, type=NiagaraPosition}}, B=<0 0 -8>)) {isSetParameters=true}
   /SetVariables(Particles.PositionOffset2:Vector3f=&Add_Vector(A=Linked() {linkedVariable {name=Particles.Position, type=NiagaraPosition}}, B=<0 0 8>)) {isSetParameters=true}
   /SetVariables(Particles.Position=&Add_Vector(A=Linked() {linkedVariable {name=Particles.Position, type=NiagaraPosition}}, B=<0 0 4000>)) {enabled=false, isSetParameters=true}
   &AddVelocity(Velocity=$User.Ribbon camera face)
   &AddVelocity(Velocity=<0 50 0>) {enabled=false, module=AddVelocity001}
   &AddVelocity(Velocity=<1 0 50>) {enabled=false, module=AddVelocity002}
  ]
  particleUpdateScript [
   &ParticleState()
   &SkeletalMeshLocation(
    Apply Sampled Orientation to Mesh Particles=true
    Bone / Socket Index=$Particles.SpawnGroup
    Bone Sampling Mode=::NewEnumerator4
    Inherit Velocity=false
    Interpolate Mesh Sampling=false
    MaskSpawnGroup=false
    Sampled Position Offset=&MakeVector(
     X=/Game/Effects/Niagara/DynamicInputs/FN_PickFloatFromSpawnGroupCutoff(Over Cutoff Float:NiagaraFloat=-50, Under Cutoff Float:NiagaraFloat=50, SpawnGroup:NiagaraInt32=$Particles.SpawnGroup, Cutoff:NiagaraInt32=1)
     Y=0
     Z=0
    )
    Sampled Position Offset Coordinate Space:ENiagaraMeshTransforms=::NewEnumerator3
    Skeletal Mesh=$User.SkeletalMesh
    Write Sampled Mesh Attrributes to Particle Attributes=true
   ) {enabled=false}
   &SolveForcesAndVelocity(Clamp Velocity=false, Limit Acceleration=false)
   &ScaleRibbonWidth(Ribbon Width Scale=&FloatFromCurve(FloatCurve=DataInterface() {propertyValues {Curve {keys [
         {arriveTangent=-1.2579, interpMode=RCIM_Cubic, time=0, value=1}
         {arriveTangent=-0.9664, interpMode=RCIM_Cubic, leaveTangent=-0.9664, tangentMode=RCTM_User, time=1, value=0}
        ]}}}))
   &ScaleColor(
    Scale Alpha=&FloatFromCurve(
     FloatCurve=DataInterface() {propertyValues {Curve {keys [<at 0 1 RCIM_Cubic>, {arriveTangent=-2.2887, interpMode=RCIM_Cubic, leaveTangent=-2.2887, tangentMode=RCTM_User, time=0.25, value=0}]}}}
     CurveIndex=$Emitter.Age
     Scale Curve=0.15
    )
    Scale RGB=&MakeVectorFromLinearColorRGB(Color=&ColorFromCurve(ColorCurve=DataInterface() {propertyValues {
        AlphaCurve {keys [
          <at 0.1755 0>
          {arriveTangent=0.6672, interpMode=RCIM_Cubic, leaveTangent=0.6672, time=0.2218, value=0.3}
          <at 0.3254 0.1>
          <at 1 0 RCIM_Cubic>
         ]}
        BlueCurve {keys [<at 0.2234 3 RCIM_Cubic>,<at 0.3386 5>,<at 0.4944 0.2812 RCIM_Cubic>]}
        GreenCurve {keys [<at 0.2234 2.1141 RCIM_Cubic>,<at 0.3386 4.38>,<at 0.4944 0.2599 RCIM_Cubic>]}
        RedCurve {keys [<at 0.2234 1.8 RCIM_Cubic>,<at 0.3386 4>,<at 0.4944 0.2531 RCIM_Cubic>]}
       }}))
    ScaleA=true
    ScaleRGB=true
   )
  ]
  rendererClasses [&NiagaraRibbonRendererProperties, &NiagaraSpriteRendererProperties]
  renderers [
   {rendererClass=&NiagaraRibbonRendererProperties}
   {rendererClass=&NiagaraRibbonRendererProperties, rendererIndex=1}
   {rendererClass=&NiagaraRibbonRendererProperties, rendererIndex=2}
   {rendererClass=&NiagaraSpriteRendererProperties, rendererIndex=3}
  ]
 }
 Claw01 {
  emitterUpdateScript [
   &EmitterState() {enabled=false}
   &EmitterState(Life Cycle Mode=::NewEnumerator1, Loop Behavior=::NewEnumerator1, Loop Duration=0.2, UseLoopDelay=false) {module=EmitterState001}
   &FN_SpawnAnimTrail_v2(Number of Anim Trails=3, Spawn Interval=0)
  ]
  particleSpawnScript [
   &FN_SetSpawnGroup(Number of Spawn Groups=1, Set Ribbon ID and Ribbon Link Order=true)
   /Game/Effects/Niagara/Modules/AnimTrail/FN_InitializeAnimTrail/FN_InitializeAnimTrailBoneIndecies(Second Bone or Socket Index=1)
   &FN_InitializeAnimTrailBoneIndecies(AnimTrail Selection=1, Second Bone or Socket Index=1) {module=FN_InitializeAnimTrailBoneIndecies001}
   &FN_InitializeAnimTrailBoneIndecies(AnimTrail Selection=2, Second Bone or Socket Index=1) {module=FN_InitializeAnimTrailBoneIndecies002}
   /Game/Effects/Niagara/Modules/AnimTrail/FN_InitializeAnimTrail/FN_InitializeAnimTrail(
    AnimTrailWidth=0.5
    Color=<rgba 0.4325 0.4494 0.5 0.05>
    Lifetime=0.1
    Position Offset:Vector3f=&SelectVectorFromArray(Direct Array Index=$Particles.SpawnGroup, Vector Selection Array=DataInterface() {propertyValues {FloatData [<0 2 0>, <0 0 -8>, <0 0 8>]}})
   )
  ]
  particleUpdateScript [
   &ScaleRibbonWidth(Ribbon Width Scale=&FloatFromCurve(FloatCurve=DataInterface() {propertyValues {Curve {keys [
         {arriveTangent=-1.2579, interpMode=RCIM_Cubic, time=0, value=1}
         {arriveTangent=-0.9664, interpMode=RCIM_Cubic, leaveTangent=-0.9664, tangentMode=RCTM_User, time=1, value=0}
        ]}}}))
   &ScaleColor(
    Scale Alpha=&Multiply_Float(
     A=&FloatFromCurve(FloatCurve=DataInterface() {propertyValues {Curve {keys [<at 0 1>,<at 1 1>]}}})
     B=&Lerp_Float(
      A=&Multiply_Float(
       A=&FloatFromCurve(
        FloatCurve=DataInterface() {propertyValues {Curve {keys [<at 0 0.25>,<at 1 0>]}}}
        Scale Curve=&Multiply_Float(
         A=2
         B=&FloatFromCurve(FloatCurve=DataInterface() {propertyValues {Curve {keys [<at 0.05 0>,<at 0.06 1>,<at 0.5 0>]}}}, CurveIndex=$Emitter.Age, Scale Curve=5)
        )
       )
       B=&FloatFromCurve(FloatCurve=DataInterface() {propertyValues {Curve {keys [<at 0 0.25>,<at 1 0>]}}}, Scale Curve=2)
      )
      B=&FloatFromCurve(
       FloatCurve=DataInterface() {propertyValues {Curve {keys [<at 0 0.25>,<at 1 0>]}}}
       Scale Curve=&Multiply_Float(
        A=2
        B=&FloatFromCurve(FloatCurve=DataInterface() {propertyValues {Curve {keys [<at 0.18 0>,<at 0.185 1>,<at 0.25 0>]}}}, CurveIndex=$Emitter.Age, Scale Curve=5)
       )
      )
      Alpha=&MakeCustomFloatFromBool(Bool=$User.SecondImpact, True Float=1, False Float=0)
     )
    )
    ScaleA=true
    ScaleRGB=true
   )
   &FN_UpdateAnimTrail_v2()
   &ScaleColor(Scale Alpha=2, ScaleA=true, ScaleRGB=true) {enabled=false, module=ScaleColor001}
  ]
  rendererClasses [&NiagaraRibbonRendererProperties]
  renderers [
   {rendererClass=&NiagaraRibbonRendererProperties}
   {rendererClass=&NiagaraRibbonRendererProperties, rendererIndex=1}
   {rendererClass=&NiagaraRibbonRendererProperties, rendererIndex=2}
   {rendererClass=&NiagaraRibbonRendererProperties, rendererIndex=3}
  ]
 }
 Claw01001 {
  bEnabled=false
  emitterUpdateScript [
   &EmitterState() {enabled=false}
   &EmitterState(Life Cycle Mode=::NewEnumerator1, Loop Behavior=::NewEnumerator1, Loop Duration=0.2, UseLoopDelay=false) {module=EmitterState001}
   &FN_SpawnAnimTrail_v2(Number of Anim Trails=3)
  ]
  particleSpawnScript [
   &FN_SetSpawnGroup(Number of Spawn Groups=1, Set Ribbon ID and Ribbon Link Order=true)
   &FN_InitializeAnimTrailBoneIndecies(Second Bone or Socket Index=1)
   &FN_InitializeAnimTrailBoneIndecies(AnimTrail Selection=1, Second Bone or Socket Index=1) {module=FN_InitializeAnimTrailBoneIndecies001}
   &FN_InitializeAnimTrailBoneIndecies(AnimTrail Selection=2, Second Bone or Socket Index=1) {module=FN_InitializeAnimTrailBoneIndecies002}
   &FN_InitializeAnimTrail(
    AnimTrailWidth=0.5
    Color=<rgba 0.4325 0.4494 0.5 0.25>
    Lifetime=0.1
    Position Offset:Vector3f=&SelectVectorFromArray(Direct Array Index=$Particles.SpawnGroup, Vector Selection Array=DataInterface() {propertyValues {FloatData [<0 2 0>, <0 0 -8>, <0 0 8>]}})
   )
  ]
  particleUpdateScript [
   &ScaleRibbonWidth(Ribbon Width Scale=&FloatFromCurve(FloatCurve=DataInterface() {propertyValues {Curve {keys [
         {arriveTangent=-1.2579, interpMode=RCIM_Cubic, time=0, value=1}
         {arriveTangent=-0.9664, interpMode=RCIM_Cubic, leaveTangent=-0.9664, tangentMode=RCTM_User, time=1, value=0}
        ]}}}))
   &ScaleColor(
    Scale Alpha=&Multiply_Float(
     A=&FloatFromCurve(FloatCurve=DataInterface() {propertyValues {Curve {keys [<at 0 1>,<at 1 1>]}}}, Scale Curve=4)
     B=&Lerp_Float(
      A=&Multiply_Float(
       A=&FloatFromCurve(
        FloatCurve=DataInterface() {propertyValues {Curve {keys [<at 0 0.25>,<at 1 0>]}}}
        Scale Curve=&Multiply_Float(
         A=2
         B=&FloatFromCurve(FloatCurve=DataInterface() {propertyValues {Curve {keys [<at 0.05 0>,<at 0.06 1>,<at 0.5 0>]}}}, CurveIndex=$Emitter.Age, Scale Curve=5)
        )
       )
       B=&FloatFromCurve(FloatCurve=DataInterface() {propertyValues {Curve {keys [<at 0 0.25>,<at 1 0>]}}}, Scale Curve=2)
      )
      B=&FloatFromCurve(
       FloatCurve=DataInterface() {propertyValues {Curve {keys [<at 0 0.25>,<at 1 0>]}}}
       Scale Curve=&Multiply_Float(
        A=2
        B=&FloatFromCurve(FloatCurve=DataInterface() {propertyValues {Curve {keys [<at 0.19 0>,<at 0.2 1>,<at 0.5 0>]}}}, CurveIndex=$Emitter.Age)
       )
      )
      Alpha=&MakeCustomFloatFromBool(Bool=$User.SecondImpact, True Float=1, False Float=0)
     )
    )
    ScaleA=true
    ScaleRGB=true
   )
   &FN_UpdateAnimTrail_v2()
   &ScaleColor(Scale Alpha=2, ScaleA=true, ScaleRGB=true) {enabled=false, module=ScaleColor001}
  ]
  rendererClasses [&NiagaraRibbonRendererProperties]
  renderers [
   {rendererClass=&NiagaraRibbonRendererProperties}
   {rendererClass=&NiagaraRibbonRendererProperties, rendererIndex=1}
   {rendererClass=&NiagaraRibbonRendererProperties, rendererIndex=2}
   {rendererClass=&NiagaraRibbonRendererProperties, rendererIndex=3}
  ]
 }
 Claw02 {
  bEnabled=false
  emitterUpdateScript [
   &EmitterState()
   &FN_SpawnAnimTrail_v2()
  ]
  particleSpawnScript [
   &FN_InitializeAnimTrailBoneIndecies(Second Bone or Socket Index=1)
   &FN_InitializeAnimTrail(AnimTrailWidth:NiagaraFloat=$User.AnimTrailWidth, Lifetime:NiagaraFloat=$User.Lifetime, Position Offset:Vector3f=<0 0 -8>)
  ]
  particleUpdateScript [
   &ScaleRibbonWidth(Ribbon Width Scale=&FloatFromCurve(FloatCurve=DataInterface() {propertyValues {Curve {keys [
         {arriveTangent=-1.2579, interpMode=RCIM_Cubic, time=0, value=1}
         {arriveTangent=-0.9664, interpMode=RCIM_Cubic, leaveTangent=-0.9664, tangentMode=RCTM_User, time=1, value=0}
        ]}}}))
   &ScaleColor(
    Scale Alpha=&FloatFromCurve(FloatCurve=DataInterface() {propertyValues {Curve {keys [<at 0 0.25>,<at 1 0>]}}})
    ScaleA=true
    ScaleRGB=true
   )
   &FN_UpdateAnimTrail_v2()
  ]
  rendererClasses [&NiagaraRibbonRendererProperties]
  renderers [{rendererClass=&NiagaraRibbonRendererProperties}]
 }
 Claw03 {
  bEnabled=false
  emitterUpdateScript [
   &EmitterState()
   &FN_SpawnAnimTrail_v2()
  ]
  particleSpawnScript [
   &FN_InitializeAnimTrailBoneIndecies(Second Bone or Socket Index=1)
   &FN_InitializeAnimTrail(AnimTrailWidth:NiagaraFloat=$User.AnimTrailWidth, Lifetime:NiagaraFloat=$User.Lifetime, Position Offset:Vector3f=<0 0 8>)
  ]
  particleUpdateScript [
   &ScaleRibbonWidth(Ribbon Width Scale=&FloatFromCurve(FloatCurve=DataInterface() {propertyValues {Curve {keys [
         {arriveTangent=-1.2579, interpMode=RCIM_Cubic, time=0, value=1}
         {arriveTangent=-0.9664, interpMode=RCIM_Cubic, leaveTangent=-0.9664, tangentMode=RCTM_User, time=1, value=0}
        ]}}}))
   &ScaleColor(
    Scale Alpha=&FloatFromCurve(FloatCurve=DataInterface() {propertyValues {Curve {keys [<at 0 0.25>,<at 1 0>]}}})
    ScaleA=true
    ScaleRGB=true
   )
   &FN_UpdateAnimTrail_v2()
  ]
  rendererClasses [&NiagaraRibbonRendererProperties]
  renderers [{rendererClass=&NiagaraRibbonRendererProperties}]
 }
 FN_Debug {
  bEnabled=false
  emitterUpdateScript [
   &EmitterState()
   &SpawnBurst_Instantaneous(Spawn Count=1, Use Spawn Probability=false)
  ]
  particleSpawnScript [&InitializeParticle(
    Lifetime=5
    Module.Write Color=true
    Module.Write Lifetime=true
    Module.Write Mass=true
    Module.Write Material Random=true
    Module.Write Position=false
    Module.Write Scale=false
    Module.Write SpriteRotation=false
    Module.Write SpriteSize=true
    Position=&SimulationPosition()
    Sprite Size=<100 100>
   )]
  particleUpdateScript [
   &ParticleState()
   /SetVariables(Particles.Debug:NiagaraFloat=$System.Timeline.NormalizedTime) {isSetParameters=true}
   &DynamicMaterialParameters(
    Index 0 Param 1=$Particles.Debug
    Module.Param0WriteEnabled=true
    Module.Param1WriteEnabled=true
    Module.Param2WriteEnabled=true
    Module.Param3WriteEnabled=true
   )
  ]
  rendererClasses [&NiagaraSpriteRendererProperties]
  renderers [{rendererClass=&NiagaraSpriteRendererProperties}]
 }
 Soft01 {
  bEnabled=false
  emitterUpdateScript [
   &EmitterState(Life Cycle Mode=::NewEnumerator1, Loop Behavior=::NewEnumerator1, Loop Duration=0.2, UseLoopDelay=false) {enabled=false}
   &EmitterState(Life Cycle Mode=::NewEnumerator1, Loop Duration=$User.Loop duration, UseLoopDelay=false) {module=EmitterState001}
   &FN_SpawnAnimTrail_v2() {enabled=false}
   &FN_SpawnAnimTrail_v2(Number of Anim Trails=3) {module=FN_SpawnAnimTrail_v2001}
  ]
  particleSpawnScript [
   &FN_SetSpawnGroup(Number of Spawn Groups=3, Set Ribbon ID and Ribbon Link Order=true)
   &FN_InitializeAnimTrailBoneIndecies(Second Bone or Socket Index=1) {enabled=false}
   &FN_InitializeAnimTrailBoneIndecies(Second Bone or Socket Index=1) {module=FN_InitializeAnimTrailBoneIndecies001}
   &FN_InitializeAnimTrailBoneIndecies(AnimTrail Selection=1, Second Bone or Socket Index=1) {module=FN_InitializeAnimTrailBoneIndecies002}
   &FN_InitializeAnimTrailBoneIndecies(AnimTrail Selection=2, Second Bone or Socket Index=1) {module=FN_InitializeAnimTrailBoneIndecies003}
   &FN_InitializeAnimTrail(
    AnimTrailWidth=0.8
    Lifetime=0.4
    Position Offset:Vector3f=&SelectVectorFromArray(Direct Array Index=$Particles.SpawnGroup, Vector Selection Array=DataInterface() {propertyValues {FloatData [<0 2 0>, <0 0 -8>, <0 0 8>]}})
   )
  ]
  particleUpdateScript [
   &ScaleRibbonWidth(Ribbon Width Scale=&FloatFromCurve(FloatCurve=DataInterface() {propertyValues {Curve {keys [
         {arriveTangent=-1.2579, interpMode=RCIM_Cubic, time=0, value=1}
         {arriveTangent=-0.9664, interpMode=RCIM_Cubic, leaveTangent=-0.9664, tangentMode=RCTM_User, time=1, value=0}
        ]}}}))
   &Color(
    Color=&ColorFromCurve(ColorCurve=DataInterface() {propertyValues {
       AlphaCurve {keys [<at 0 0>, {arriveTangent=0.3173, interpMode=RCIM_Cubic, time=0.15, value=0.3}, <at 0.3254 0.1>, <at 1.0518 0 RCIM_Cubic>]}
       BlueCurve {keys [<at 0.2234 1 RCIM_Cubic>,<at 0.3386 1>]}
       GreenCurve {keys [<at 0.2234 1 RCIM_Cubic>,<at 0.3386 1>]}
       RedCurve {keys [<at 0.2234 1 RCIM_Cubic>,<at 0.3386 1>]}
      }})
    Scale Alpha=&FloatFromCurve(FloatCurve=DataInterface() {propertyValues {Curve {keys [<at 0 0.25>,<at 0.2 0>]}}}, CurveIndex=$Emitter.Age)
   )
   &ScaleColor(
    Scale Alpha=&FloatFromCurve(FloatCurve=DataInterface() {propertyValues {Curve {keys [<at 0 0.25>,<at 0.13 0>]}}}, CurveIndex=$Emitter.Age)
    ScaleA=true
    ScaleRGB=true
   ) {enabled=false, module=ScaleColor003}
   &FN_UpdateAnimTrail_v2()
  ]
  rendererClasses [&NiagaraRibbonRendererProperties]
  renderers [{rendererClass=&NiagaraRibbonRendererProperties}]
 }
 Soft02 {
  bEnabled=false
  emitterUpdateScript [
   &EmitterState()
   &FN_SpawnAnimTrail_v2()
  ]
  particleSpawnScript [
   &FN_InitializeAnimTrailBoneIndecies(Second Bone or Socket Index=1)
   &FN_InitializeAnimTrail(AnimTrailWidth=0.8, Lifetime=0.13, Position Offset:Vector3f=<0 0 -8>)
  ]
  particleUpdateScript [
   &ScaleRibbonWidth(Ribbon Width Scale=&FloatFromCurve(FloatCurve=DataInterface() {propertyValues {Curve {keys [
         {arriveTangent=-1.2579, interpMode=RCIM_Cubic, time=0, value=1}
         {arriveTangent=-0.9664, interpMode=RCIM_Cubic, leaveTangent=-0.9664, tangentMode=RCTM_User, time=1, value=0}
        ]}}}))
   &Color(
    Color=&ColorFromCurve(ColorCurve=DataInterface() {propertyValues {
       AlphaCurve {keys [<at 0 0>, {arriveTangent=0.3173, interpMode=RCIM_Cubic, time=0.15, value=0.3}, <at 0.3254 0.1>, <at 1.0518 0 RCIM_Cubic>]}
       BlueCurve {keys [<at 0.2234 1 RCIM_Cubic>,<at 0.3386 1>]}
       GreenCurve {keys [<at 0.2234 1 RCIM_Cubic>,<at 0.3386 1>]}
       RedCurve {keys [<at 0.2234 1 RCIM_Cubic>,<at 0.3386 1>]}
      }})
    Scale Alpha=1.5
   )
   &FN_UpdateAnimTrail_v2()
  ]
  rendererClasses [&NiagaraRibbonRendererProperties]
  renderers [{rendererClass=&NiagaraRibbonRendererProperties}]
 }
 Soft03 {
  bEnabled=false
  emitterUpdateScript [
   &EmitterState()
   &FN_SpawnAnimTrail_v2()
  ]
  particleSpawnScript [
   &FN_InitializeAnimTrailBoneIndecies(Second Bone or Socket Index=1)
   &FN_InitializeAnimTrail(AnimTrailWidth=0.8, Lifetime=0.13, Position Offset:Vector3f=<0 0 8>)
  ]
  particleUpdateScript [
   &ScaleRibbonWidth(Ribbon Width Scale=&FloatFromCurve(FloatCurve=DataInterface() {propertyValues {Curve {keys [
         {arriveTangent=-1.2579, interpMode=RCIM_Cubic, time=0, value=1}
         {arriveTangent=-0.9664, interpMode=RCIM_Cubic, leaveTangent=-0.9664, tangentMode=RCTM_User, time=1, value=0}
        ]}}}))
   &Color(
    Color=&ColorFromCurve(ColorCurve=DataInterface() {propertyValues {
       AlphaCurve {keys [<at 0 0>, {arriveTangent=0.3173, interpMode=RCIM_Cubic, time=0.15, value=0.3}, <at 0.3254 0.1>, <at 1.0518 0 RCIM_Cubic>]}
       BlueCurve {keys [<at 0.2234 1 RCIM_Cubic>,<at 0.3386 1>]}
       GreenCurve {keys [<at 0.2234 1 RCIM_Cubic>,<at 0.3386 1>]}
       RedCurve {keys [<at 0.2234 1 RCIM_Cubic>,<at 0.3386 1>]}
      }})
    Scale Alpha=1.5
   )
   &FN_UpdateAnimTrail_v2()
  ]
  rendererClasses [&NiagaraRibbonRendererProperties]
  renderers [{rendererClass=&NiagaraRibbonRendererProperties}]
 }
}
summary {
 systemName=NS_DesertShadowMale_AnimTrails
 emitters [
  {emitterName=Claw01, rendererClasses [&NiagaraRibbonRendererProperties]}
  {bEnabled=false, emitterName=Soft01, rendererClasses [&NiagaraRibbonRendererProperties]}
  {bEnabled=false, emitterName=Claw02, rendererClasses [&NiagaraRibbonRendererProperties]}
  {bEnabled=false, emitterName=Claw03, rendererClasses [&NiagaraRibbonRendererProperties]}
  {bEnabled=false, emitterName=Soft02, rendererClasses [&NiagaraRibbonRendererProperties]}
  {bEnabled=false, emitterName=Soft03, rendererClasses [&NiagaraRibbonRendererProperties]}
  {bEnabled=false, emitterName=FN_Debug, rendererClasses [&NiagaraSpriteRendererProperties]}
  {
   bEnabled=false
   emitterName=Blades_Trails002
   rendererClasses [&NiagaraRibbonRendererProperties, &NiagaraSpriteRendererProperties]
  }
  {bEnabled=false, emitterName=Claw01001, rendererClasses [&NiagaraRibbonRendererProperties]}
 ]
 userVariables [
  User.AnimTrailWidth {type=NiagaraFloat, defaultValue=1, description=""}
  User.Lifetime {type=NiagaraFloat, defaultValue=0.15, description=""}
  User.Loop duration {type=NiagaraFloat, defaultValue=0.25, description=""}
  User.NewNiagaraDataInterfaceSkeletalMesh {
   type=NiagaraDataInterfaceSkeletalMesh
   defaultValue {
    struct=NiagaraExt_VariableValue_DataInterface
    value {dataInterface=&self:NiagaraDataInterfaceSkeletalMesh_2}
   }
  }
  User.Ribbon camera face {type=Vector3f, description="", defaultValue {struct=Vector3f, value=<-0.1 0 0.1>}}
  User.RibbonIDOffset {type=NiagaraInt32}
  User.SecondImpact {type=NiagaraBool}
  User.SkeletalMesh {
   type=NiagaraDataInterfaceSkeletalMesh
   defaultValue {
    struct=NiagaraExt_VariableValue_DataInterface
    value {dataInterface=&self:NiagaraDataInterfaceSkeletalMesh_0}
   }
  }
  User.SkeletalMesh_R {
   type=NiagaraDataInterfaceSkeletalMesh
   defaultValue {
    struct=NiagaraExt_VariableValue_DataInterface
    value {dataInterface=&self:NiagaraDataInterfaceSkeletalMesh_1}
   }
  }
 ]
}
